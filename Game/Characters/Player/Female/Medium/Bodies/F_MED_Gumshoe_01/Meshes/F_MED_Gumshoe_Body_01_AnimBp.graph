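# Animation Blueprint for the female medium Gumshoe player body that copies a source mesh's pose and adds secondary motion to the tie and lock bones.
# On initialize it caches the owner's skeletal mesh component as the Copy Pose From Mesh source.
# Anim Dynamics then simulates dyn_tie fully and dyn_lock_02 at partial blend.
# A final Fortnite Flap node partially blends inverted left-thigh rotation onto dyn_lock_01, avoiding simulation for that bone.
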AnimBlueprint /Game/Characters/Player/Female/Medium/Bodies/F_MED_Gumshoe_01/Meshes/F_MED_Gumshoe_Body_01_AnimBp
target_skeleton=&/F_MED_Gumshoe_Body_01_Skeleton
graphs [
 AnimGraph {nodes [
   n1 = Get MeshToCopy() -> (MeshToCopy:Skeletal Mesh Component Object Reference)
   n2 = Anim Dynamics - Bone: dyn_tie(ComponentPose=Local To Component(LocalPose=Copy Pose From Mesh(SourceMeshComponent=@n1)), Alpha=1, ExternalForce="0,0,0") {
    angularBiasOverride=1
    angularDampingOverride=1
    bGravityOverrideInSimSpace=true
    bOverrideAngularDamping=true
    bOverrideLinearDamping=true
    lODThreshold=1
    numSolverIterationsPostUpdate=4
    numSolverIterationsPreUpdate=16
    alphaBoolBlend {blendOption=Linear}
    boundBone {boneName=dyn_tie}
    physicsBodyDefinitions [{
      sphereCollisionRadius=0
      boundBone {boneName=dyn_tie}
      boxExtents=<10 8 2>
      constraintSetup {angularTargetAxis=AxisZ, twistAxis=AxisZ, angularLimitsMax=<40 45 0>, angularLimitsMin=<-40 -45 0>, angularTarget=<0 0 0>}
      localJointOffset=<0 0 -8>
     }]
   }
   n3 = Anim Dynamics - Bone: dyn_lock_02(ComponentPose=@n2, Alpha=0.3, ExternalForce="0,0,0") {
    angularBiasOverride=1
    angularDampingOverride=1
    bGravityOverrideInSimSpace=true
    bOverrideAngularDamping=true
    bOverrideLinearDamping=true
    lODThreshold=1
    numSolverIterationsPostUpdate=4
    numSolverIterationsPreUpdate=16
    alphaBoolBlend {blendOption=Linear}
    boundBone {boneName=dyn_lock_02}
    physicsBodyDefinitions [{
      sphereCollisionRadius=0
      boundBone {boneName=dyn_lock_02}
      boxExtents=<10 8 2>
      constraintSetup {angularTargetAxis=AxisZ, twistAxis=AxisZ, angularLimitsMax=<50 50 0>, angularLimitsMin=<-50 -50 0>, angularTarget=<0 0 0>}
      localJointOffset=<0 0 -8>
     }]
   }
   n4 = Flap Node - Bone: dyn_lock_01!(ComponentPose:Component Space Pose Link Structure=@n3, Alpha:float=0.7) -> (Pose:Component Space Pose Link Structure) {
    is=title FortAnimGraphNode_Flap
    targetClampMax=3.14
    targetComponent=RotationY
    alphaBoolBlend {blendOption=Linear}
    sourceDrives [{
      bUseRange=true
      drivingCurve=null
      multiplier=-1
      rangeMax=1.7453
      rangeMin=-1.7453
      remappedMax=1.7453
      remappedMin=-1.7453
      sourceComponent=RotationZ
      useQuaternion=true
      sourceBone {boneName=thigh_l}
     }]
    targetBone {boneName=dyn_lock_01}
   }
   Output Pose(Result=Component To Local(ComponentPose=@n4)) {layerGroup=None}
  ]}
 EventGraph {nodes [
   Event BlueprintInitializeAnimation()
   > Set MeshToCopy(MeshToCopy:Skeletal Mesh Component Object Reference=GetOwnerSkeletalMeshComponent()) -> (Output_Get:Skeletal Mesh Component Object Reference)
   Event BlueprintUpdateAnimation()
  ]}
]
variables [MeshToCopy {type=SkeletalMeshComponent}]
